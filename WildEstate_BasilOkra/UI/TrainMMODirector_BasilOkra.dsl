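; TrainMMODirector_BasilOkra is a dynamic UI director that adds the TrainMMOScene_BasilOkra UI scene for players in a zone.
; On BeginPlay it gets the owning local player controller's Fort global UI context and adds the scene only if IsinZone is true.
; The check runs once at startup; Tick and the construction script are empty.
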(event EventBeginPlay
  (if (GlobalUIContext|IsinZone (PlayerController|LocalPlayerSubsystems|GetFortGlobalUIContext (DynamicUIDirector|GetOwningLocalPlayerController)))
    (DynamicUIDirector|AddScene "/WildEstate_BasilOkra/UI/TrainMMOScene_BasilOkra.TrainMMOScene_BasilOkra")))

(event EventTick
  (params (DeltaSeconds :type Float%20%28single-precision%29)))

(fn ConstructionScript)
